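# Physics asset for the FaunaPike medium male body's jacket, defining collision and joints for its dynamic bones.
# Five capsule chains (front left/right, sides, rear middle) hang from base bodies, with swing and twist limits loosening toward the tips.
# Angularly free cross-constraints tie both side chains to the rear at the third link.
# Pelvis and leg capsules, plus a large slope-warp box, provide body and likely ground collision shapes.
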PhysicsAsset /BRCosmetics/Characters/Player/Male/Medium/Bodies/M_MED_FaunaPike/Meshes/M_MED_FaunaPike_Jacket_Physics
bodies {
 calf_l [{shapeType=Capsule, length=39.5651, radius=9.6721, center=<-22.8529 0.4514 0.6388>, rotation=<rot 0 87.2615 -89.9715>}]
 calf_r [{shapeType=Capsule, length=39.5651, radius=9.6721, center=<22.8529 -0.4514 -0.6388>, rotation=<rot 0 87.2615 90.0285>}]
 def_slope_warp [{
   shapeType=Box
   extentX=286.2154
   extentY=278.7956
   extentZ=33.447
   shapeName=def_slope_warp_box
   center=<0 0 -14.0088>
   rotation=<rot 0 -90 0>
  }]
 dyn_jacket_front_l_01 [{shapeType=Capsule, length=35.0198, radius=8, shapeName=dyn_jacket_front_l_01_capsule, center=<4 -11.9483 0>}]
 dyn_jacket_front_l_02 [{shapeType=Capsule, length=35.0198, radius=8, shapeName=dyn_jacket_front_l_02_capsule, center=<4 -8.9487 0>}]
 dyn_jacket_front_l_03 [{shapeType=Capsule, length=35.0198, radius=8, shapeName=dyn_jacket_front_l_03_capsule, center=<4 -5.7345 0>}]
 dyn_jacket_front_l_04 [{shapeType=Capsule, length=35.0198, radius=8, shapeName=dyn_jacket_front_l_04_capsule, center=<4 -5.2482 0>}]
 dyn_jacket_front_l_05 [{shapeType=Capsule, length=35.0198, radius=8, shapeName=dyn_jacket_front_l_05_capsule, center=<5.1862 -3.113 0>}]
 dyn_jacket_front_l_base [{shapeType=Capsule, length=0.505, radius=0.505, shapeName=dyn_jacket_front_l_base_capsule, rotation=<rot 0 -90 0>}]
 dyn_jacket_front_r_01 [{shapeType=Capsule, length=35.0198, radius=8, shapeName=dyn_jacket_front_r_01_capsule, center=<-4 11.9483 0>, rotation=<rot 0 0 -180>}]
 dyn_jacket_front_r_02 [{shapeType=Capsule, length=35.0198, radius=8, shapeName=dyn_jacket_front_r_02_capsule, center=<-4 8.9487 0>, rotation=<rot 0 0 -180>}]
 dyn_jacket_front_r_03 [{shapeType=Capsule, length=35.0198, radius=8, shapeName=dyn_jacket_front_r_03_capsule, center=<-4 5.7097 0>, rotation=<rot 0 0 -180>}]
 dyn_jacket_front_r_04 [{shapeType=Capsule, length=35.0198, radius=8, shapeName=dyn_jacket_front_r_04_capsule, center=<-4 6.2414 0>, rotation=<rot 0 0 -180>}]
 dyn_jacket_front_r_05 [{shapeType=Capsule, length=35.0198, radius=8, shapeName=dyn_jacket_front_r_05_capsule, center=<-4 3.5647 0>, rotation=<rot 0 0 -180>}]
 dyn_jacket_front_r_base [{shapeType=Capsule, length=0.505, radius=0.505, shapeName=dyn_jacket_front_r_base_capsule, rotation=<rot 0 -90 0>}]
 dyn_jacket_rear_mid_01 [{shapeType=Capsule, length=35.0198, radius=8, shapeName=dyn_jacket_rear_mid_01_capsule, center=<4 -6.5752 0>}]
 dyn_jacket_rear_mid_02 [{shapeType=Capsule, length=35.0198, radius=8, shapeName=dyn_jacket_rear_mid_02_capsule, center=<4 -5.4407 0>}]
 dyn_jacket_rear_mid_03 [{shapeType=Capsule, length=35.0198, radius=8, shapeName=dyn_jacket_rear_mid_03_capsule, center=<4 -3.8602 0>}]
 dyn_jacket_rear_mid_04 [{shapeType=Capsule, length=35.0198, radius=8, shapeName=dyn_jacket_rear_mid_04_capsule, center=<4 -2.1476 0>}]
 dyn_jacket_rear_mid_05 [{shapeType=Capsule, length=35.0198, radius=8, shapeName=dyn_jacket_rear_mid_05_capsule, center=<4 -1.6378 0>}]
 dyn_jacket_rear_mid_base [{shapeType=Capsule, length=0.505, radius=0.505, shapeName=dyn_jacket_rear_mid_base_capsule, rotation=<rot 0 -90 0>}]
 dyn_jacket_side_l_01 [{shapeType=Capsule, length=35.0198, radius=8, shapeName=dyn_jacket_side_l_01_capsule, center=<4 -9.2566 0>}]
 dyn_jacket_side_l_02 [{shapeType=Capsule, length=35.0198, radius=8, shapeName=dyn_jacket_side_l_02_capsule, center=<4 -7.7788 0>}]
 dyn_jacket_side_l_03 [{shapeType=Capsule, length=35.0198, radius=8, shapeName=dyn_jacket_side_l_03_capsule, center=<4 -6.5351 0>}]
 dyn_jacket_side_l_04 [
  {shapeType=Capsule, length=34.9497, radius=8.0701, shapeName=dyn_jacket_side_l_04_capsule, center=<4 -4.6114 0>}
  {
   shapeType=Capsule
   length=20.4661
   radius=9.5339
   shapeName=dyn_jacket_side_l_04_capsule
   center=<0 2.6472 -17.3534>
   rotation=<rot 0 0.0001 -30.9375>
  }
 ]
 dyn_jacket_side_l_05 [
  {shapeType=Capsule, length=35.0594, radius=7.9604, shapeName=dyn_jacket_side_l_05_capsule, center=<4 -2.9743 0>}
  {
   shapeType=Capsule
   length=24.4759
   radius=11.3566
   shapeName=dyn_jacket_side_l_05_capsule
   center=<1.3216 3.8257 -18.9011>
   rotation=<rot 0.0001 0.0002 -33.75>
  }
 ]
 dyn_jacket_side_l_base [{shapeType=Capsule, length=0.505, radius=0.505, shapeName=dyn_jacket_side_l_base_capsule, rotation=<rot 0 -90 0>}]
 dyn_jacket_side_r_01 [{shapeType=Capsule, length=35.0198, radius=8, shapeName=dyn_jacket_side_r_01_capsule, center=<-4 9.2566 0>, rotation=<rot 0 0 -180>}]
 dyn_jacket_side_r_02 [{shapeType=Capsule, length=35.0198, radius=8, shapeName=dyn_jacket_side_r_02_capsule, center=<-4 8.1637 0>, rotation=<rot 0 0 -180>}]
 dyn_jacket_side_r_03 [{shapeType=Capsule, length=35.0198, radius=8, shapeName=dyn_jacket_side_r_03_capsule, center=<-4 6.5061 0>, rotation=<rot 0 0 -180>}]
 dyn_jacket_side_r_04 [
  {shapeType=Capsule, length=34.9497, radius=8.0701, shapeName=dyn_jacket_side_r_04_capsule, center=<-4 4.6114 0>, rotation=<rot 0 0 -180>}
  {
   shapeType=Capsule
   length=20.4661
   radius=9.5339
   shapeName=dyn_jacket_side_r_04_capsule
   center=<0 -2.6472 17.3534>
   rotation=<rot 0 0.0001 149.0625>
  }
 ]
 dyn_jacket_side_r_05 [
  {shapeType=Capsule, length=35.0594, radius=7.9604, shapeName=dyn_jacket_side_r_05_capsule, center=<-4 2.9743 0>, rotation=<rot 0 0 -180>}
  {
   shapeType=Capsule
   length=24.4759
   radius=11.3566
   shapeName=dyn_jacket_side_r_05_capsule
   center=<-1.3216 -3.8257 18.9011>
   rotation=<rot 0.0001 0.0002 146.25>
  }
 ]
 dyn_jacket_side_r_base [{shapeType=Capsule, length=0.505, radius=0.505, shapeName=dyn_jacket_side_r_base_capsule, rotation=<rot 0 -90 0>}]
 foot_l [{shapeType=Capsule, length=11.4428, radius=16.7956, center=<-3.8599 8.751 0.06>, rotation=<rot 0 1.1007 -87.9209>}]
 foot_r [{shapeType=Capsule, length=11.4428, radius=16.7956, center=<3.8599 -8.751 -0.06>, rotation=<rot 0 1.1007 92.0791>}]
 pelvis [{shapeType=Capsule, length=12, radius=8}]
 thigh_l [{shapeType=Capsule, length=30.7212, radius=9.5009, center=<-22.2774 -1.4091 1.8389>, rotation=<rot 0 -86.0775 -90.0925>}]
 thigh_r [{shapeType=Capsule, length=30.7212, radius=9.5009, center=<22.2774 1.4091 -1.8389>, rotation=<rot 0 -86.0775 89.9075>}]
}
constraints [
 {bone1Name=dyn_jacket_front_l_01, bone2Name=dyn_jacket_front_l_base, swing1LimitDegrees=14, swing2LimitDegrees=8, twistLimitDegrees=8}
 {bone1Name=dyn_jacket_front_l_02, bone2Name=dyn_jacket_front_l_01, swing1LimitDegrees=14, swing2LimitDegrees=8, twistLimitDegrees=8}
 {bone1Name=dyn_jacket_front_l_03, bone2Name=dyn_jacket_front_l_02, swing1LimitDegrees=16, swing2LimitDegrees=9, twistLimitDegrees=9}
 {bone1Name=dyn_jacket_front_l_04, bone2Name=dyn_jacket_front_l_03, swing1LimitDegrees=20, swing2LimitDegrees=10, twistLimitDegrees=10}
 {bone1Name=dyn_jacket_front_l_05, bone2Name=dyn_jacket_front_l_04, swing1LimitDegrees=24, swing2LimitDegrees=12, twistLimitDegrees=12}
 {bone1Name=dyn_jacket_front_r_01, bone2Name=dyn_jacket_front_r_base, swing1LimitDegrees=14, swing2LimitDegrees=8, twistLimitDegrees=8}
 {bone1Name=dyn_jacket_front_r_02, bone2Name=dyn_jacket_front_r_01, swing1LimitDegrees=14, swing2LimitDegrees=8, twistLimitDegrees=8}
 {bone1Name=dyn_jacket_front_r_03, bone2Name=dyn_jacket_front_r_02, swing1LimitDegrees=16, swing2LimitDegrees=9, twistLimitDegrees=9}
 {bone1Name=dyn_jacket_front_r_04, bone2Name=dyn_jacket_front_r_03, swing1LimitDegrees=20, swing2LimitDegrees=10, twistLimitDegrees=10}
 {bone1Name=dyn_jacket_front_r_05, bone2Name=dyn_jacket_front_r_04, swing1LimitDegrees=24, swing2LimitDegrees=12, twistLimitDegrees=12}
 {bone1Name=dyn_jacket_rear_mid_01, bone2Name=dyn_jacket_rear_mid_base, swing1LimitDegrees=14, swing2LimitDegrees=8, twistLimitDegrees=8}
 {bone1Name=dyn_jacket_rear_mid_02, bone2Name=dyn_jacket_rear_mid_01, swing1LimitDegrees=14, swing2LimitDegrees=8, twistLimitDegrees=8}
 {bone1Name=dyn_jacket_rear_mid_03, bone2Name=dyn_jacket_rear_mid_02, swing1LimitDegrees=16, swing2LimitDegrees=9, twistLimitDegrees=9}
 {bone1Name=dyn_jacket_rear_mid_04, bone2Name=dyn_jacket_rear_mid_03, swing1LimitDegrees=20, swing2LimitDegrees=10, twistLimitDegrees=10}
 {bone1Name=dyn_jacket_rear_mid_05, bone2Name=dyn_jacket_rear_mid_04, swing1LimitDegrees=24, swing2LimitDegrees=12, twistLimitDegrees=12}
 {bone1Name=dyn_jacket_side_l_01, bone2Name=dyn_jacket_side_l_base, swing1LimitDegrees=14, swing2LimitDegrees=8, twistLimitDegrees=8}
 {bone1Name=dyn_jacket_side_l_02, bone2Name=dyn_jacket_side_l_01, swing1LimitDegrees=14, swing2LimitDegrees=8, twistLimitDegrees=8}
 {bone1Name=dyn_jacket_side_l_03, bone2Name=dyn_jacket_side_l_02, swing1LimitDegrees=16, swing2LimitDegrees=9, twistLimitDegrees=9}
 {bone1Name=dyn_jacket_side_l_04, bone2Name=dyn_jacket_side_l_03, swing1LimitDegrees=20, swing2LimitDegrees=10, twistLimitDegrees=10}
 {bone1Name=dyn_jacket_side_l_05, bone2Name=dyn_jacket_side_l_04, swing1LimitDegrees=24, swing2LimitDegrees=12, twistLimitDegrees=12}
 {bone1Name=dyn_jacket_side_r_01, bone2Name=dyn_jacket_side_r_base, swing1LimitDegrees=14, swing2LimitDegrees=8, twistLimitDegrees=8}
 {bone1Name=dyn_jacket_side_r_02, bone2Name=dyn_jacket_side_r_01, swing1LimitDegrees=14, swing2LimitDegrees=8, twistLimitDegrees=8}
 {bone1Name=dyn_jacket_side_r_03, bone2Name=dyn_jacket_side_r_02, swing1LimitDegrees=16, swing2LimitDegrees=9, twistLimitDegrees=9}
 {bone1Name=dyn_jacket_side_r_04, bone2Name=dyn_jacket_side_r_03, swing1LimitDegrees=20, swing2LimitDegrees=10, twistLimitDegrees=10}
 {bone1Name=dyn_jacket_side_r_05, bone2Name=dyn_jacket_side_r_04, swing1LimitDegrees=24, swing2LimitDegrees=12, twistLimitDegrees=12}
 {
  bone1Name=dyn_jacket_side_l_03
  bone2Name=dyn_jacket_rear_mid_03
  swing1LimitDegrees=45
  swing1Motion=Free
  swing2LimitDegrees=45
  swing2Motion=Free
  twistLimitDegrees=45
  twistMotion=Free
 }
 {
  bone1Name=dyn_jacket_side_r_03
  bone2Name=dyn_jacket_rear_mid_03
  swing1LimitDegrees=45
  swing1Motion=Free
  swing2LimitDegrees=45
  swing2Motion=Free
  twistLimitDegrees=45
  twistMotion=Free
 }
]
