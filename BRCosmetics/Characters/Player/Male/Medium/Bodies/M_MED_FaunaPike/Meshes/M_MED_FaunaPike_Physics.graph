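# Physics asset for the M_MED_FaunaPike male player body cosmetic, covering its dynamic hood, two hood drawstrings and chain with capsule bodies.
# Each strand is a constraint chain from a tiny base body, with angular limits generally widening toward the tip.
# Capsules on the spine, clavicles and pecs add upper-torso volumes.
# The pec bodies link to spine_03 with free swing and twist, while the other torso bodies have no constraints between them.
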PhysicsAsset /BRCosmetics/Characters/Player/Male/Medium/Bodies/M_MED_FaunaPike/Meshes/M_MED_FaunaPike_Physics
bodies {
 clavicle_l [{shapeType=Capsule, length=11.6898, radius=7.9327, center=<4.0198 -0.7547 -0.9751>, rotation=<rot 87.6854 -0.0017 -0.0017>}]
 clavicle_r [{shapeType=Capsule, length=11.6898, radius=7.9327, center=<-4.0198 0.7547 0.9751>, rotation=<rot 87.6854 -0.0017 179.9983>}]
 deform_pec_l [{
   shapeType=Capsule
   length=5.5888
   radius=6.6347
   shapeName=deform_pec_l_capsule
   center=<-0.627 -5.8584 -1.3089>
   rotation=<rot -1.2903 117.769 -90.6093>
  }]
 deform_pec_r [{
   shapeType=Capsule
   length=5.5888
   radius=6.6347
   shapeName=deform_pec_r_capsule
   center=<0.627 5.8584 1.3089>
   rotation=<rot -1.2903 117.769 89.3907>
  }]
 dyn_chain_01 [{shapeType=Capsule, length=8.3815, radius=2.7832, shapeName=dyn_chain_01_capsule, center=<1.9589 -2.5745 0>, rotation=<rot 0 -90 0>}]
 dyn_chain_02 [{shapeType=Capsule, length=6.8419, radius=2.4434, shapeName=dyn_chain_02_capsule, center=<2.0243 -1.214 0.1252>, rotation=<rot 0 -90 0>}]
 dyn_chain_03 [{shapeType=Capsule, length=6.1535, radius=3.0218, shapeName=dyn_chain_03_capsule, center=<0.9337 0.1981 0.1485>, rotation=<rot 0 -90 0>}]
 dyn_chain_base [{shapeType=Capsule, length=0.505, radius=0.505, shapeName=dyn_chain_base_capsule, rotation=<rot 0 -90 0>}]
 dyn_hood_01 [{shapeType=Capsule, length=12, radius=5, shapeName=dyn_hood_01_capsule, center=<2 -4.3661 0>}]
 dyn_hood_02 [{shapeType=Capsule, length=12, radius=5, shapeName=dyn_hood_02_capsule, center=<2 -4.261 0>}]
 dyn_hood_03 [{shapeType=Capsule, length=12, radius=5, shapeName=dyn_hood_03_capsule, center=<3.7003 -2.6326 0>}]
 dyn_hood_base [{shapeType=Capsule, length=0.505, radius=0.505, shapeName=dyn_hood_base_capsule, rotation=<rot 0 -90 0>}]
 dyn_hood_string_l_01 [{shapeType=Capsule, length=6, radius=2, shapeName=dyn_hood_string_l_01_capsule, center=<3.5 0 0>, rotation=<rot 90 4.5045 0>}]
 dyn_hood_string_l_02 [{shapeType=Capsule, length=6, radius=2, shapeName=dyn_hood_string_l_02_capsule, center=<3.5 0 0>, rotation=<rot 90 0 0>}]
 dyn_hood_string_l_03 [{shapeType=Capsule, length=5.5235, radius=2.4765, shapeName=dyn_hood_string_l_03_capsule, center=<3.5 0 0>, rotation=<rot 90 0 0>}]
 dyn_hood_string_l_04 [{shapeType=Capsule, length=6.7826, radius=2.3294, shapeName=dyn_hood_string_l_04_capsule, center=<3.5 0 0>, rotation=<rot 90 4.4926 0>}]
 dyn_hood_string_l_base [{shapeType=Capsule, length=0.505, radius=0.505, shapeName=dyn_hood_string_l_base_capsule, rotation=<rot 0 -90 0>}]
 dyn_hood_string_r_01 [{shapeType=Capsule, length=6, radius=2, shapeName=dyn_hood_string_r_01_capsule, center=<-3.5 0 0>, rotation=<rot 90 -175.4955 0>}]
 dyn_hood_string_r_02 [{shapeType=Capsule, length=6, radius=2, shapeName=dyn_hood_string_r_02_capsule, center=<-3.5 0 0>, rotation=<rot 90 180 0>}]
 dyn_hood_string_r_03 [{shapeType=Capsule, length=5.6309, radius=2.3691, shapeName=dyn_hood_string_r_03_capsule, center=<-3.5 0 0>, rotation=<rot 90 180 0>}]
 dyn_hood_string_r_04 [{shapeType=Capsule, length=6.7826, radius=2.3294, shapeName=dyn_hood_string_r_04_capsule, center=<-3.5 0 0>, rotation=<rot 90 -175.5074 0>}]
 dyn_hood_string_r_base [{shapeType=Capsule, length=0.505, radius=0.505, shapeName=dyn_hood_string_r_base_capsule, rotation=<rot 0 -90 0>}]
 spine_03 [{
   shapeType=Capsule
   length=6.604
   radius=11.9533
   shapeName=spine_03_capsule
   center=<7.24 -0.9006 -0.041>
   rotation=<rot 0.0422 -87.1803 -89.9946>
  }]
 spine_05 [{shapeType=Capsule, length=7.9094, radius=11.11, center=<-0.5622 -1.4077 0>}]
}
constraints [
 {bone1Name=dyn_hood_string_l_01, bone2Name=dyn_hood_string_l_base, swing1LimitDegrees=8, swing2LimitDegrees=8, twistLimitDegrees=8}
 {bone1Name=dyn_hood_string_l_02, bone2Name=dyn_hood_string_l_01, swing1LimitDegrees=12, swing2LimitDegrees=12, twistLimitDegrees=12}
 {bone1Name=dyn_hood_string_l_03, bone2Name=dyn_hood_string_l_02, swing1LimitDegrees=24, swing2LimitDegrees=24, twistLimitDegrees=24}
 {bone1Name=dyn_hood_string_l_04, bone2Name=dyn_hood_string_l_03, swing1LimitDegrees=32, swing2LimitDegrees=32, twistLimitDegrees=32}
 {bone1Name=dyn_hood_string_r_01, bone2Name=dyn_hood_string_r_base, swing1LimitDegrees=8, swing2LimitDegrees=8, twistLimitDegrees=8}
 {bone1Name=dyn_hood_string_r_02, bone2Name=dyn_hood_string_r_01, swing1LimitDegrees=12, swing2LimitDegrees=12, twistLimitDegrees=12}
 {bone1Name=dyn_hood_string_r_03, bone2Name=dyn_hood_string_r_02, swing1LimitDegrees=24, swing2LimitDegrees=24, twistLimitDegrees=24}
 {bone1Name=dyn_hood_string_r_04, bone2Name=dyn_hood_string_r_03, swing1LimitDegrees=32, swing2LimitDegrees=32, twistLimitDegrees=32}
 {bone1Name=dyn_hood_01, bone2Name=dyn_hood_base, swing1LimitDegrees=14, swing2LimitDegrees=14, twistLimitDegrees=14}
 {bone1Name=dyn_hood_02, bone2Name=dyn_hood_01, swing1LimitDegrees=16, swing2LimitDegrees=16, twistLimitDegrees=16}
 {bone1Name=dyn_hood_03, bone2Name=dyn_hood_02, swing1LimitDegrees=38, swing2LimitDegrees=38, twistLimitDegrees=38}
 {bone1Name=dyn_chain_01, bone2Name=dyn_chain_base, swing1LimitDegrees=2, swing2LimitDegrees=2, twistLimitDegrees=2}
 {bone1Name=dyn_chain_02, bone2Name=dyn_chain_01, swing1LimitDegrees=8, swing2LimitDegrees=16, twistLimitDegrees=8}
 {bone1Name=dyn_chain_03, bone2Name=dyn_chain_02, swing1LimitDegrees=24, swing2LimitDegrees=45, twistLimitDegrees=8}
 {
  bone1Name=deform_pec_l
  bone2Name=spine_03
  swing1LimitDegrees=45
  swing1Motion=Free
  swing2LimitDegrees=45
  swing2Motion=Free
  twistLimitDegrees=45
  twistMotion=Free
 }
 {
  bone1Name=deform_pec_r
  bone2Name=spine_03
  swing1LimitDegrees=45
  swing1Motion=Free
  swing2LimitDegrees=45
  swing2Motion=Free
  twistLimitDegrees=45
  twistMotion=Free
 }
]
