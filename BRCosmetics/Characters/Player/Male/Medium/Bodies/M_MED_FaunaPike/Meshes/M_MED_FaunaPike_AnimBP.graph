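# Post-process-style animation Blueprint for the FaunaPike male body cosmetic, adding secondary motion to hood, strings, chain, sleeves and jacket.
# Copied parent pose runs through Control Rig, slope-warp bone, RigidBody, translation constraints and per-bone sleeve AnimDynamics.
# A jacket RigidBody pass substitutes when cloth is disabled.
# Updates drive cloth controller, gravity override, interpolated slope warp and front-end sleeve alphas; emote events toggle cloth simulation off and on.
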AnimBlueprint /BRCosmetics/Characters/Player/Male/Medium/Bodies/M_MED_FaunaPike/Meshes/M_MED_FaunaPike_AnimBP
target_skeleton=&/M_MED_FaunaPike_Skeleton
graphs [
 AnimGraph {nodes [
   n1 = RigidBody!(
    ComponentPose=Local To Component(LocalPose=Use cached pose 'Post Default Dynamics'())
    Alpha
    OverrideWorldGravity:Vector="0,0,-980"
    ExternalForce
   ) -> (Pose) {
    bOverrideWorldGravity=true
    overridePhysicsAsset=&/M_MED_FaunaPike_Jacket_Physics
    simulationSpace=BaseBoneSpace
    baseBoneRef {boneName=root}
    componentAppliedLinearAccClamp=<0 0 0>
    overrideWorldGravity {z=-980}
    simSpaceSettings {
     maxAngularAcceleration=0.5
     maxAngularVelocity=2
     maxLinearAcceleration=1200
     maxLinearVelocity=500
     velocityScaleZ=0.2
     worldAlpha=1
     externalLinearDragV=<1.8 1.8 1>
    }
   }
   Output Pose(Result=Blend Poses by bool(BlendPose_0=Use cached pose 'Post Default Dynamics'(), BlendPose_1=Component To Local(ComponentPose=@n1), bActiveValue=Get bClothEnabled())) {layerGroup=None}
   n2 = "Transform (Modify) Bone - Bone: def_slope_warp"!(
    ComponentPose=Local To Component(LocalPose=Use cached pose 'Post Deform'())
    bAlphaBoolEnabled:bool=True
    Translation="0,0,0"
    Rotation="0,0,0"
    Scale="1,1,1"
   ) -> (Pose) {
    alphaInputType=Bool
    rotationMode=BMM_Replace
    rotationSpace=BCS_WorldSpace
    translationMode=BMM_Replace
    translationSpace=BCS_BoneSpace
    alphaBoolBlend {blendOption=Linear}
    boneToModify {boneName=def_slope_warp}
   }
   n3 = RigidBody(ComponentPose=@n2, OverrideWorldGravity="0,0,-980") {
    bOverrideWorldGravity=true
    overridePhysicsAsset=&/M_MED_FaunaPike_Physics
    simulationSpace=BaseBoneSpace
    baseBoneRef {boneName=root}
    componentAppliedLinearAccClamp=<0 0 0>
    overrideWorldGravity {z=-980}
    simSpaceSettings {
     maxAngularAcceleration=8
     maxAngularVelocity=16
     maxLinearAcceleration=160
     maxLinearVelocity=1400
     worldAlpha=1
     externalLinearDragV=<0.2 0.2 0.2>
    }
   }
   n4 = "Constraint - dyn_hood_string_l_01 (T)"(ComponentPose=@n3, Alpha=1, ConstraintWeights_0=1) {
    boneToModify {boneName=dyn_hood_string_l_01}
    constraintSetup [{transformType=Translation, targetBone {boneName=dyn_hood_string_l_base}}]
    constraintWeights [1]
   }
   n5 = "Constraint - dyn_hood_string_r_01 (T)"(ComponentPose=@n4, Alpha=1, ConstraintWeights_0=1) {
    boneToModify {boneName=dyn_hood_string_r_01}
    constraintSetup [{transformType=Translation, targetBone {boneName=dyn_hood_string_r_base}}]
    constraintWeights [1]
   }
   n6 = "Constraint - dyn_chain_01 (T)"(ComponentPose=@n5, Alpha=1, ConstraintWeights_0=1) {
    boneToModify {boneName=dyn_chain_01}
    constraintSetup [{transformType=Translation, targetBone {boneName=dyn_chain_base}}]
    constraintWeights [1]
   }
   n7 = "Constraint - dyn_hood_01 (T)"(ComponentPose=@n6, Alpha=1, ConstraintWeights_0=1) {
    boneToModify {boneName=dyn_hood_01}
    constraintSetup [{transformType=Translation, targetBone {boneName=dyn_hood_base}}]
    constraintWeights [1]
   }
   n8 = Anim Dynamics - Bone: dyn_sleeve_l_base(ComponentPose=@n7, Alpha=0.8, ExternalForce="0,0,0") {
    alpha=0.8
    bLinearSpring=true
    bOverrideLinearDamping=true
    bUseGravityOverride=true
    gravityScale=0.35
    linearDampingOverride=0.8
    linearSpringConstant=20
    simulationSpace=BoneRelative
    boundBone {boneName=dyn_sleeve_l_base}
    componentAppliedLinearAccClamp=<1400 1400 1400>
    gravityOverride {z=-980}
    physicsBodyDefinitions [{
      boundBone {boneName=dyn_sleeve_l_base}
      boxExtents=<50 20 20>
      constraintSetup {angularLimitsMax=<0 0 0>, angularLimitsMin=<0 0 0>, angularTarget=<1 0 0>, linearAxesMax=<2.5 0 0>}
      localJointOffset=<4 0 0>
     }]
    relativeSpaceBone {boneName=lowerarm_l}
    simSpaceSettings {maxAngularAcceleration=40, maxAngularVelocity=300, simSpaceAngularAlpha=1}
   }
   n9 = Anim Dynamics - Bone: dyn_sleeve_r_base(ComponentPose=@n8, Alpha=0.8, ExternalForce="0,0,0") {
    alpha=0.8
    bLinearSpring=true
    bOverrideLinearDamping=true
    bUseGravityOverride=true
    gravityScale=0.35
    linearDampingOverride=0.8
    linearSpringConstant=20
    simulationSpace=BoneRelative
    boundBone {boneName=dyn_sleeve_r_base}
    componentAppliedLinearAccClamp=<1400 1400 1400>
    gravityOverride {z=-980}
    physicsBodyDefinitions [{
      boundBone {boneName=dyn_sleeve_r_base}
      boxExtents=<50 20 20>
      constraintSetup {angularLimitsMax=<0 0 0>, angularLimitsMin=<0 0 0>, angularTarget=<1 0 0>, linearAxesMax=<2.5 0 0>}
      localJointOffset=<-4 0 0>
     }]
    relativeSpaceBone {boneName=lowerarm_r}
    simSpaceSettings {maxAngularAcceleration=40, maxAngularVelocity=300, simSpaceAngularAlpha=1}
   }
   n10 = Anim Dynamics - Bone: dyn_sleeve_l_upper_01!(ComponentPose=@n9, ExternalForce="0,0,0") -> (Pose) {
    alpha=0.6
    angularDampingOverride=1
    angularSpringConstant=25000
    bAngularSpring=true
    bOverrideAngularDamping=true
    bUseGravityOverride=true
    gravityScale=0.35
    simulationSpace=BoneRelative
    boundBone {boneName=dyn_sleeve_l_upper_01}
    componentAppliedLinearAccClamp=<1400 1400 1400>
    componentLinearAccScale=<1 1 1>
    componentLinearVelScale=<1 1 1>
    gravityOverride {z=-980}
    physicsBodyDefinitions [{
      boundBone {boneName=dyn_sleeve_l_upper_01}
      boxExtents=<20 30 30>
      constraintSetup {angularLimitsMax=<5 0 5>, angularLimitsMin=<-5 -10 -5>, angularTarget=<1 0 0>}
      localJointOffset=<4 0 0>
     }]
    relativeSpaceBone {boneName=spine_05}
   }
   n11 = Anim Dynamics - Bone: dyn_sleeve_l_upper_02!(ComponentPose=@n10, ExternalForce="0,0,0") -> (Pose) {
    alpha=0.6
    angularDampingOverride=1
    angularSpringConstant=25000
    bAngularSpring=true
    bOverrideAngularDamping=true
    bUseGravityOverride=true
    gravityScale=0.35
    simulationSpace=BoneRelative
    boundBone {boneName=dyn_sleeve_l_upper_02}
    componentAppliedLinearAccClamp=<1400 1400 1400>
    componentLinearAccScale=<1 1 1>
    componentLinearVelScale=<1 1 1>
    gravityOverride {z=-980}
    physicsBodyDefinitions [{
      boundBone {boneName=dyn_sleeve_l_upper_02}
      boxExtents=<20 30 30>
      constraintSetup {angularLimitsMax=<5 0 5>, angularLimitsMin=<-5 -10 -5>, angularTarget=<1 0 0>}
      localJointOffset=<4 0 0>
     }]
    relativeSpaceBone {boneName=spine_05}
   }
   n12 = Anim Dynamics - Bone: dyn_sleeve_l_lower_01!(ComponentPose=@n11, ExternalForce="0,0,0") -> (Pose) {
    alpha=0.6
    angularDampingOverride=1
    angularSpringConstant=25000
    bAngularSpring=true
    bOverrideAngularDamping=true
    bUseGravityOverride=true
    gravityScale=0.35
    simulationSpace=BoneRelative
    boundBone {boneName=dyn_sleeve_l_lower_01}
    componentAppliedLinearAccClamp=<1400 1400 1400>
    componentLinearAccScale=<1 1 1>
    componentLinearVelScale=<1 1 1>
    gravityOverride {z=-980}
    physicsBodyDefinitions [{
      boundBone {boneName=dyn_sleeve_l_lower_01}
      boxExtents=<20 30 30>
      constraintSetup {angularLimitsMax=<5 0 5>, angularLimitsMin=<-5 -10 -5>, angularTarget=<1 0 0>}
      localJointOffset=<4 0 0>
     }]
    relativeSpaceBone {boneName=spine_05}
   }
   n13 = Anim Dynamics - Bone: dyn_sleeve_l_lower_02!(ComponentPose=@n12, ExternalForce="0,0,0") -> (Pose) {
    alpha=0.6
    angularDampingOverride=1
    angularSpringConstant=25000
    bAngularSpring=true
    bOverrideAngularDamping=true
    bUseGravityOverride=true
    gravityScale=0.35
    simulationSpace=BoneRelative
    boundBone {boneName=dyn_sleeve_l_lower_02}
    componentAppliedLinearAccClamp=<1400 1400 1400>
    componentLinearAccScale=<1 1 1>
    componentLinearVelScale=<1 1 1>
    gravityOverride {z=-980}
    physicsBodyDefinitions [{
      boundBone {boneName=dyn_sleeve_l_lower_02}
      boxExtents=<20 30 30>
      constraintSetup {angularLimitsMax=<5 0 5>, angularLimitsMin=<-5 -10 -5>, angularTarget=<1 0 0>}
      localJointOffset=<4 0 0>
     }]
    relativeSpaceBone {boneName=spine_05}
   }
   n14 = Anim Dynamics - Bone: dyn_sleeve_l_mid_behind_01!(ComponentPose=@n13, ExternalForce="0,0,0") -> (Pose) {
    alpha=0.6
    angularDampingOverride=1
    angularSpringConstant=25000
    bAngularSpring=true
    bOverrideAngularDamping=true
    bUseGravityOverride=true
    gravityScale=0.35
    simulationSpace=BoneRelative
    boundBone {boneName=dyn_sleeve_l_mid_behind_01}
    componentAppliedLinearAccClamp=<1400 1400 1400>
    componentLinearAccScale=<1 1 1>
    componentLinearVelScale=<1 1 1>
    gravityOverride {z=-980}
    physicsBodyDefinitions [{
      boundBone {boneName=dyn_sleeve_l_mid_behind_01}
      boxExtents=<20 30 30>
      constraintSetup {angularLimitsMax=<5 0 5>, angularLimitsMin=<-5 -10 -5>, angularTarget=<1 0 0>}
      localJointOffset=<4 0 0>
     }]
    relativeSpaceBone {boneName=spine_05}
   }
   n15 = Anim Dynamics - Bone: dyn_sleeve_l_mid_behind_02!(ComponentPose=@n14, ExternalForce="0,0,0") -> (Pose) {
    alpha=0.6
    angularDampingOverride=1
    angularSpringConstant=25000
    bAngularSpring=true
    bOverrideAngularDamping=true
    bUseGravityOverride=true
    gravityScale=0.35
    simulationSpace=BoneRelative
    boundBone {boneName=dyn_sleeve_l_mid_behind_02}
    componentAppliedLinearAccClamp=<1400 1400 1400>
    componentLinearAccScale=<1 1 1>
    componentLinearVelScale=<1 1 1>
    gravityOverride {z=-980}
    physicsBodyDefinitions [{
      boundBone {boneName=dyn_sleeve_l_mid_behind_02}
      boxExtents=<20 30 30>
      constraintSetup {angularLimitsMax=<5 0 5>, angularLimitsMin=<-5 -10 -5>, angularTarget=<1 0 0>}
      localJointOffset=<4 0 0>
     }]
    relativeSpaceBone {boneName=spine_05}
   }
   n16 = Anim Dynamics - Bone: dyn_sleeve_l_mid_front_01!(ComponentPose=@n15, ExternalForce="0,0,0") -> (Pose) {
    alpha=0.6
    angularDampingOverride=1
    angularSpringConstant=25000
    bAngularSpring=true
    bOverrideAngularDamping=true
    bUseGravityOverride=true
    gravityScale=0.35
    simulationSpace=BoneRelative
    boundBone {boneName=dyn_sleeve_l_mid_front_01}
    componentAppliedLinearAccClamp=<1400 1400 1400>
    componentLinearAccScale=<1 1 1>
    componentLinearVelScale=<1 1 1>
    gravityOverride {z=-980}
    physicsBodyDefinitions [{
      boundBone {boneName=dyn_sleeve_l_mid_front_01}
      boxExtents=<20 30 30>
      constraintSetup {angularLimitsMax=<5 0 5>, angularLimitsMin=<-5 -10 -5>, angularTarget=<1 0 0>}
      localJointOffset=<4 0 0>
     }]
    relativeSpaceBone {boneName=spine_05}
   }
   n17 = Anim Dynamics - Bone: dyn_sleeve_l_mid_front_02!(ComponentPose=@n16, ExternalForce="0,0,0") -> (Pose) {
    alpha=0.6
    angularDampingOverride=1
    angularSpringConstant=25000
    bAngularSpring=true
    bOverrideAngularDamping=true
    bUseGravityOverride=true
    gravityScale=0.35
    simulationSpace=BoneRelative
    boundBone {boneName=dyn_sleeve_l_mid_front_02}
    componentAppliedLinearAccClamp=<1400 1400 1400>
    componentLinearAccScale=<1 1 1>
    componentLinearVelScale=<1 1 1>
    gravityOverride {z=-980}
    physicsBodyDefinitions [{
      boundBone {boneName=dyn_sleeve_l_mid_front_02}
      boxExtents=<20 30 30>
      constraintSetup {angularLimitsMax=<5 0 5>, angularLimitsMin=<-5 -10 -5>, angularTarget=<1 0 0>}
      localJointOffset=<4 0 0>
     }]
    relativeSpaceBone {boneName=spine_05}
   }
   n18 = Anim Dynamics - Bone: dyn_sleeve_r_upper_01!(ComponentPose=@n17, ExternalForce="0,0,0") -> (Pose) {
    alpha=0.6
    angularDampingOverride=1
    angularSpringConstant=25000
    bAngularSpring=true
    bOverrideAngularDamping=true
    bUseGravityOverride=true
    gravityScale=0.35
    simulationSpace=BoneRelative
    boundBone {boneName=dyn_sleeve_r_upper_01}
    componentAppliedLinearAccClamp=<1400 1400 1400>
    componentLinearAccScale=<1 1 1>
    componentLinearVelScale=<1 1 1>
    gravityOverride {z=-980}
    physicsBodyDefinitions [{
      boundBone {boneName=dyn_sleeve_r_upper_01}
      boxExtents=<20 30 30>
      constraintSetup {angularLimitsMax=<5 0 5>, angularLimitsMin=<-5 -10 -5>, angularTarget=<1 0 0>}
      localJointOffset=<-4 0 0>
     }]
    relativeSpaceBone {boneName=spine_05}
   }
   n19 = Anim Dynamics - Bone: dyn_sleeve_r_upper_02!(ComponentPose=@n18, ExternalForce="0,0,0") -> (Pose) {
    alpha=0.6
    angularDampingOverride=1
    angularSpringConstant=25000
    bAngularSpring=true
    bOverrideAngularDamping=true
    bUseGravityOverride=true
    gravityScale=0.35
    simulationSpace=BoneRelative
    boundBone {boneName=dyn_sleeve_r_upper_02}
    componentAppliedLinearAccClamp=<1400 1400 1400>
    componentLinearAccScale=<1 1 1>
    componentLinearVelScale=<1 1 1>
    gravityOverride {z=-980}
    physicsBodyDefinitions [{
      boundBone {boneName=dyn_sleeve_r_upper_02}
      boxExtents=<20 30 30>
      constraintSetup {angularLimitsMax=<5 0 5>, angularLimitsMin=<-5 -10 -5>, angularTarget=<1 0 0>}
      localJointOffset=<-4 0 0>
     }]
    relativeSpaceBone {boneName=spine_05}
   }
   n20 = Anim Dynamics - Bone: dyn_sleeve_r_lower_01!(ComponentPose=@n19, ExternalForce="0,0,0") -> (Pose) {
    alpha=0.6
    angularDampingOverride=1
    angularSpringConstant=25000
    bAngularSpring=true
    bOverrideAngularDamping=true
    bUseGravityOverride=true
    gravityScale=0.35
    simulationSpace=BoneRelative
    boundBone {boneName=dyn_sleeve_r_lower_01}
    componentAppliedLinearAccClamp=<1400 1400 1400>
    componentLinearAccScale=<1 1 1>
    componentLinearVelScale=<1 1 1>
    gravityOverride {z=-980}
    physicsBodyDefinitions [{
      boundBone {boneName=dyn_sleeve_r_lower_01}
      boxExtents=<20 30 30>
      constraintSetup {angularLimitsMax=<5 0 5>, angularLimitsMin=<-5 -10 -5>, angularTarget=<1 0 0>}
      localJointOffset=<-4 0 0>
     }]
    relativeSpaceBone {boneName=spine_05}
   }
   n21 = Anim Dynamics - Bone: dyn_sleeve_r_lower_02!(ComponentPose=@n20, ExternalForce="0,0,0") -> (Pose) {
    alpha=0.6
    angularDampingOverride=1
    angularSpringConstant=25000
    bAngularSpring=true
    bOverrideAngularDamping=true
    bUseGravityOverride=true
    gravityScale=0.35
    simulationSpace=BoneRelative
    boundBone {boneName=dyn_sleeve_r_lower_02}
    componentAppliedLinearAccClamp=<1400 1400 1400>
    componentLinearAccScale=<1 1 1>
    componentLinearVelScale=<1 1 1>
    gravityOverride {z=-980}
    physicsBodyDefinitions [{
      boundBone {boneName=dyn_sleeve_r_lower_02}
      boxExtents=<20 30 30>
      constraintSetup {angularLimitsMax=<5 0 5>, angularLimitsMin=<-5 -10 -5>, angularTarget=<1 0 0>}
      localJointOffset=<-4 0 0>
     }]
    relativeSpaceBone {boneName=spine_05}
   }
   n22 = Anim Dynamics - Bone: dyn_sleeve_r_mid_behind_01!(ComponentPose=@n21, ExternalForce="0,0,0") -> (Pose) {
    alpha=0.6
    angularDampingOverride=1
    angularSpringConstant=25000
    bAngularSpring=true
    bOverrideAngularDamping=true
    bUseGravityOverride=true
    gravityScale=0.35
    simulationSpace=BoneRelative
    boundBone {boneName=dyn_sleeve_r_mid_behind_01}
    componentAppliedLinearAccClamp=<1400 1400 1400>
    componentLinearAccScale=<1 1 1>
    componentLinearVelScale=<1 1 1>
    gravityOverride {z=-980}
    physicsBodyDefinitions [{
      boundBone {boneName=dyn_sleeve_r_mid_behind_01}
      boxExtents=<20 30 30>
      constraintSetup {angularLimitsMax=<5 0 5>, angularLimitsMin=<-5 -10 -5>, angularTarget=<1 0 0>}
      localJointOffset=<-4 0 0>
     }]
    relativeSpaceBone {boneName=spine_05}
   }
   n23 = Anim Dynamics - Bone: dyn_sleeve_r_mid_behind_02!(ComponentPose=@n22, ExternalForce="0,0,0") -> (Pose) {
    alpha=0.6
    angularDampingOverride=1
    angularSpringConstant=25000
    bAngularSpring=true
    bOverrideAngularDamping=true
    bUseGravityOverride=true
    gravityScale=0.35
    simulationSpace=BoneRelative
    boundBone {boneName=dyn_sleeve_r_mid_behind_02}
    componentAppliedLinearAccClamp=<1400 1400 1400>
    componentLinearAccScale=<1 1 1>
    componentLinearVelScale=<1 1 1>
    gravityOverride {z=-980}
    physicsBodyDefinitions [{
      boundBone {boneName=dyn_sleeve_r_mid_behind_02}
      boxExtents=<20 30 30>
      constraintSetup {angularLimitsMax=<5 0 5>, angularLimitsMin=<-5 -10 -5>, angularTarget=<1 0 0>}
      localJointOffset=<-4 0 0>
     }]
    relativeSpaceBone {boneName=spine_05}
   }
   n24 = Anim Dynamics - Bone: dyn_sleeve_r_mid_front_01!(ComponentPose=@n23, ExternalForce="0,0,0") -> (Pose) {
    alpha=0.6
    angularDampingOverride=1
    angularSpringConstant=25000
    bAngularSpring=true
    bOverrideAngularDamping=true
    bUseGravityOverride=true
    gravityScale=0.35
    simulationSpace=BoneRelative
    boundBone {boneName=dyn_sleeve_r_mid_front_01}
    componentAppliedLinearAccClamp=<1400 1400 1400>
    componentLinearAccScale=<1 1 1>
    componentLinearVelScale=<1 1 1>
    gravityOverride {z=-980}
    physicsBodyDefinitions [{
      boundBone {boneName=dyn_sleeve_r_mid_front_01}
      boxExtents=<20 30 30>
      constraintSetup {angularLimitsMax=<5 0 5>, angularLimitsMin=<-5 -10 -5>, angularTarget=<1 0 0>}
      localJointOffset=<-4 0 0>
     }]
    relativeSpaceBone {boneName=spine_05}
   }
   n25 = Anim Dynamics - Bone: dyn_sleeve_r_mid_front_02!(ComponentPose=@n24, ExternalForce="0,0,0") -> (Pose) {
    alpha=0.6
    angularDampingOverride=1
    angularSpringConstant=25000
    bAngularSpring=true
    bOverrideAngularDamping=true
    bUseGravityOverride=true
    gravityScale=0.35
    simulationSpace=BoneRelative
    boundBone {boneName=dyn_sleeve_r_mid_front_02}
    componentAppliedLinearAccClamp=<1400 1400 1400>
    componentLinearAccScale=<1 1 1>
    componentLinearVelScale=<1 1 1>
    gravityOverride {z=-980}
    physicsBodyDefinitions [{
      boundBone {boneName=dyn_sleeve_r_mid_front_02}
      boxExtents=<20 30 30>
      constraintSetup {angularLimitsMax=<5 0 5>, angularLimitsMin=<-5 -10 -5>, angularTarget=<1 0 0>}
      localJointOffset=<-4 0 0>
     }]
    relativeSpaceBone {boneName=spine_05}
   }
   Save cached pose 'Post Default Dynamics'(Pose=Component To Local(ComponentPose=@n25)) {cacheName=Post Default Dynamics}
   n26 = Control Rig!(Source=Copy Pose From Mesh() {bUseAttachedParent=true}, Alpha, ClothEnabled:bool=Get bClothEnabled(), PlayerSpeed2D:double=Get PawnSpeed2D()) -> (Pose) {controlRigAssetReference {controlRigAsset=&/M_MED_FaunaPike_CtrlRig}}
   Save cached pose 'Post Deform'(Pose=@n26) {cacheName=Post Deform}
  ]}
 EventGraph {nodes [
   n1 = Event BlueprintUpdateAnimation()
   > Set Delta Time X(Delta Time X=@n1.DeltaTimeX)
   > n2 = Sequence!(execute) -> (then_0, then_1, then_2:Exec, then_3:Exec)
   > Branch(Condition=Get bClothEnabled())
   > Cloth Asset Controller Function!(
    execute
    CharacterPartAnimInstance:Custom Character Part Anim Instance Object Reference=Self-Reference()
    ClothAssetControllerParameters:Fort Cloth Asset Controller State Parameter Modifiers Object Reference=/BRCosmetics/Characters/Player/Common/Cloth/Common_ClothAssetDataParams
    ClothingAssetName:Name=M_MED_FaunaPike_Clothing
    TrackedBoneName:Name=pelvis
    JumpFallingDurationSeconds:float=1.3
   ) -> (then, OutAppliedParameters:Cloth Asset Controller Parameters Structure)
   Branch(execute=@n2.then_1, Condition=Get bRigidBodyEnabled())
   > n3 = "Gravity Override State Function (Parameter Asset) (Nativized)"(
    CharacterPartAnimInstance=Self-Reference()
    GameTimeInSeconds=GetGameTimeInSeconds()
    Parameters=&/M_MED_FaunaPike_Additional_parameters
   )
   > Set Gravity Override(Gravity Override=@n3)
   Handle Slope Warp!(execute=@n2.then_2, self, DeltaTime:double=Get Delta Time X()) -> (then)
   n4 = Branch(execute=@n2.then_3, Condition=Get bIsInFrontEnd())
   > Set FrontendSleeveAlphas(FrontendSleeveAlphas:double=0.2) -> (Output_Get:double)
   Set FrontendSleeveAlphas(execute=@n4.else, FrontendSleeveAlphas=0.8)
   Event OnEmoteStarted!() -> (OutputDelegate, then, MontageItemDef:Fort Item Definition Object Reference) {is=title K2Node_Event}
   n5 = GetPartSkeletalMeshComponent(self=Self-Reference())
   > Set ClothBlendWeight!(execute, ClothBlendWeight:float=0, self:Skeletal Mesh Component Object Reference=@n5) -> (then, Output_Get:float)
   > Set bDisableClothSimulation!(execute, bDisableClothSimulation:bool=true, self:Skeletal Mesh Component Object Reference=@n5) -> (then, Output_Get:bool)
   Event OnEndPlayingEmote() {is=title K2Node_Event}
   > Set ClothBlendWeight(ClothBlendWeight=1, self=@n5)
   > Set bDisableClothSimulation(bDisableClothSimulation=false, self=@n5)
  ]}
 Handle Slope Warp {nodes [
   n1 = Handle Slope Warp!() -> (then, DeltaTime:double=~) {is=title K2Node_FunctionEntry}
   > n2 = Sequence!(execute) -> (then_0, then_1)
   n3 = Get bIsOnGround() -> (bIsOnGround:bool)
   > n4 = Branch(Condition=OR Boolean!(A=NOT Boolean(A=@n3), B=Get bIsPlayingEmote(), C:bool=Get bIsInVehicle() -> (bIsInVehicle:bool)) -> (ReturnValue))
   > Set Target Info Input Root Slope Translation(Target Info Input Root Slope Translation:Vector="0,0,-5000") -> (Output_Get:Vector)
   > Set Slope Info Input Slope Rotation(Slope Info Input Slope Rotation:Rotator="0, 0, 0") -> (Output_Get:Rotator)
   Set Slope Info Input Slope Rotation(
    execute=@n4.else
    Slope Info Input Slope Rotation=RInterpTo!(
     Current:Rotator=Get Slope Info Input Slope Rotation() -> (Slope Info Input Slope Rotation:Rotator)
     Target:Rotator=Get!() -> (
      SlopeInfoInput_RootSlopeTranslation:Vector=~
      SlopeInfoInput_LeftFootSlopeTranslation:Vector=~
      SlopeInfoInput_RightFootSlopeTranslation:Vector=~
      SlopeInfoInput_SlopeRotation:Rotator=~
     ).SlopeInfoInput_SlopeRotation
     DeltaTime:float=@n1
     InterpSpeed:float=7
    ) -> (ReturnValue:Rotator)
   )
   > n5 = Branch(Condition=Get bClothEnabled())
   > Set Target Info Input Root Slope Translation(Target Info Input Root Slope Translation="0,0,-3")
   Set Target Info Input Root Slope Translation(execute=@n5.else, Target Info Input Root Slope Translation="0,0,-5")
   Set Slope Info Input Root Slope Translation(
    execute=@n2.then_1
    Slope Info Input Root Slope Translation:Vector=VInterpTo!(
     Current:Vector=Get Slope Info Input Root Slope Translation() -> (Slope Info Input Root Slope Translation:Vector)
     Target:Vector=Get Target Info Input Root Slope Translation() -> (Target Info Input Root Slope Translation:Vector)
     DeltaTime:float=@n1
     InterpSpeed:float=10
    ) -> (ReturnValue:Vector)
   ) -> (Output_Get:Vector)
  ]}
]
variables [
 Delta Time X {type=number}
 FrontendSleeveAlphas {type=number}
 Gravity Override {type=Vector}
 Slope Info Input Root Slope Translation {type=Vector}
 Slope Info Input Slope Rotation {type=Rotator}
 Target Info Input Root Slope Translation {type=Vector}
]
