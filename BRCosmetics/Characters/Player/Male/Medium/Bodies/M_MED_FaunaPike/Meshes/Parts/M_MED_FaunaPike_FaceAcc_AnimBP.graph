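# Animation Blueprint for the FaunaPike male character's face accessory mesh, adding subtle secondary motion to its dyn_hair bones.
# Copies the pose from the attached parent mesh, then runs a series of single-bone AnimDynamics nodes across hair strands a through n.
# All nodes share identical root-relative settings: stiff angular springs, high damping, reduced gravity and tight angular limits.
# The update event has no logic.
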AnimBlueprint /BRCosmetics/Characters/Player/Male/Medium/Bodies/M_MED_FaunaPike/Meshes/Parts/M_MED_FaunaPike_FaceAcc_AnimBP
target_skeleton=&/M_MED_FaunaPike_FaceAcc_Skeleton
graphs [
 AnimGraph {nodes [
   n1 = Anim Dynamics - Bone: dyn_hair_a_01(ComponentPose=Local To Component(LocalPose=Copy Pose From Mesh() {bUseAttachedParent=true}), Alpha=1, ExternalForce="0,0,0") {
    angularDampingOverride=0.99
    angularSpringConstant=10000
    bAngularSpring=true
    bOverrideAngularDamping=true
    bUseGravityOverride=true
    bUsePlanarLimit=false
    gravityScale=0.35
    simulationSpace=RootRelative
    boundBone {boneName=dyn_hair_a_01}
    componentAppliedLinearAccClamp=<1400 1400 1400>
    componentLinearAccScale=<1 1 1>
    componentLinearVelScale=<1 1 1>
    gravityOverride {z=-980}
    physicsBodyDefinitions [{
      boundBone {boneName=dyn_hair_a_01}
      boxExtents=<20 30 30>
      constraintSetup {angularLimitsMax=<3 3 3>, angularLimitsMin=<-3 -3 -3>, angularTarget=<1 0 0>}
      localJointOffset=<4 0 0>
     }]
    simSpaceSettings {maxAngularAcceleration=30, maxAngularVelocity=300, simSpaceAngularAlpha=1}
   }
   n2 = Anim Dynamics - Bone: dyn_hair_b_01(ComponentPose=@n1, Alpha=1, ExternalForce="0,0,0") {
    angularDampingOverride=0.99
    angularSpringConstant=10000
    bAngularSpring=true
    bOverrideAngularDamping=true
    bUseGravityOverride=true
    bUsePlanarLimit=false
    gravityScale=0.35
    simulationSpace=RootRelative
    boundBone {boneName=dyn_hair_b_01}
    componentAppliedLinearAccClamp=<1400 1400 1400>
    componentLinearAccScale=<1 1 1>
    componentLinearVelScale=<1 1 1>
    gravityOverride {z=-980}
    physicsBodyDefinitions [{
      boundBone {boneName=dyn_hair_b_01}
      boxExtents=<20 30 30>
      constraintSetup {angularLimitsMax=<3 3 3>, angularLimitsMin=<-3 -3 -3>, angularTarget=<1 0 0>}
      localJointOffset=<4 0 0>
     }]
    simSpaceSettings {maxAngularAcceleration=30, maxAngularVelocity=300, simSpaceAngularAlpha=1}
   }
   n3 = Anim Dynamics - Bone: dyn_hair_b_02(ComponentPose=@n2, Alpha=1, ExternalForce="0,0,0") {
    angularDampingOverride=0.99
    angularSpringConstant=10000
    bAngularSpring=true
    bOverrideAngularDamping=true
    bUseGravityOverride=true
    bUsePlanarLimit=false
    gravityScale=0.35
    simulationSpace=RootRelative
    boundBone {boneName=dyn_hair_b_02}
    componentAppliedLinearAccClamp=<1400 1400 1400>
    componentLinearAccScale=<1 1 1>
    componentLinearVelScale=<1 1 1>
    gravityOverride {z=-980}
    physicsBodyDefinitions [{
      boundBone {boneName=dyn_hair_b_02}
      boxExtents=<20 30 30>
      constraintSetup {angularLimitsMax=<3 3 3>, angularLimitsMin=<-3 -3 -3>, angularTarget=<1 0 0>}
      localJointOffset=<4 0 0>
     }]
    simSpaceSettings {maxAngularAcceleration=30, maxAngularVelocity=300, simSpaceAngularAlpha=1}
   }
   n4 = Anim Dynamics - Bone: dyn_hair_c_01(ComponentPose=@n3, Alpha=1, ExternalForce="0,0,0") {
    angularDampingOverride=0.99
    angularSpringConstant=10000
    bAngularSpring=true
    bOverrideAngularDamping=true
    bUseGravityOverride=true
    bUsePlanarLimit=false
    gravityScale=0.35
    simulationSpace=RootRelative
    boundBone {boneName=dyn_hair_c_01}
    componentAppliedLinearAccClamp=<1400 1400 1400>
    componentLinearAccScale=<1 1 1>
    componentLinearVelScale=<1 1 1>
    gravityOverride {z=-980}
    physicsBodyDefinitions [{
      boundBone {boneName=dyn_hair_c_01}
      boxExtents=<20 30 30>
      constraintSetup {angularLimitsMax=<3 3 3>, angularLimitsMin=<-3 -3 -3>, angularTarget=<1 0 0>}
      localJointOffset=<4 0 0>
     }]
    simSpaceSettings {maxAngularAcceleration=30, maxAngularVelocity=300, simSpaceAngularAlpha=1}
   }
   n5 = Anim Dynamics - Bone: dyn_hair_c_02(ComponentPose=@n4, Alpha=1, ExternalForce="0,0,0") {
    angularDampingOverride=0.99
    angularSpringConstant=10000
    bAngularSpring=true
    bOverrideAngularDamping=true
    bUseGravityOverride=true
    bUsePlanarLimit=false
    gravityScale=0.35
    simulationSpace=RootRelative
    boundBone {boneName=dyn_hair_c_02}
    componentAppliedLinearAccClamp=<1400 1400 1400>
    componentLinearAccScale=<1 1 1>
    componentLinearVelScale=<1 1 1>
    gravityOverride {z=-980}
    physicsBodyDefinitions [{
      boundBone {boneName=dyn_hair_c_02}
      boxExtents=<20 30 30>
      constraintSetup {angularLimitsMax=<3 3 3>, angularLimitsMin=<-3 -3 -3>, angularTarget=<1 0 0>}
      localJointOffset=<4 0 0>
     }]
    simSpaceSettings {maxAngularAcceleration=30, maxAngularVelocity=300, simSpaceAngularAlpha=1}
   }
   n6 = Anim Dynamics - Bone: dyn_hair_d_01(ComponentPose=@n5, Alpha=1, ExternalForce="0,0,0") {
    angularDampingOverride=0.99
    angularSpringConstant=10000
    bAngularSpring=true
    bOverrideAngularDamping=true
    bUseGravityOverride=true
    bUsePlanarLimit=false
    gravityScale=0.35
    simulationSpace=RootRelative
    boundBone {boneName=dyn_hair_d_01}
    componentAppliedLinearAccClamp=<1400 1400 1400>
    componentLinearAccScale=<1 1 1>
    componentLinearVelScale=<1 1 1>
    gravityOverride {z=-980}
    physicsBodyDefinitions [{
      boundBone {boneName=dyn_hair_d_01}
      boxExtents=<20 30 30>
      constraintSetup {angularLimitsMax=<3 3 3>, angularLimitsMin=<-3 -3 -3>, angularTarget=<1 0 0>}
      localJointOffset=<4 0 0>
     }]
    simSpaceSettings {maxAngularAcceleration=30, maxAngularVelocity=300, simSpaceAngularAlpha=1}
   }
   n7 = Anim Dynamics - Bone: dyn_hair_d_02(ComponentPose=@n6, Alpha=1, ExternalForce="0,0,0") {
    angularDampingOverride=0.99
    angularSpringConstant=10000
    bAngularSpring=true
    bOverrideAngularDamping=true
    bUseGravityOverride=true
    bUsePlanarLimit=false
    gravityScale=0.35
    simulationSpace=RootRelative
    boundBone {boneName=dyn_hair_d_02}
    componentAppliedLinearAccClamp=<1400 1400 1400>
    componentLinearAccScale=<1 1 1>
    componentLinearVelScale=<1 1 1>
    gravityOverride {z=-980}
    physicsBodyDefinitions [{
      boundBone {boneName=dyn_hair_d_02}
      boxExtents=<20 30 30>
      constraintSetup {angularLimitsMax=<3 3 3>, angularLimitsMin=<-3 -3 -3>, angularTarget=<1 0 0>}
      localJointOffset=<4 0 0>
     }]
    simSpaceSettings {maxAngularAcceleration=30, maxAngularVelocity=300, simSpaceAngularAlpha=1}
   }
   n8 = Anim Dynamics - Bone: dyn_hair_d_03(ComponentPose=@n7, Alpha=1, ExternalForce="0,0,0") {
    angularDampingOverride=0.99
    angularSpringConstant=10000
    bAngularSpring=true
    bOverrideAngularDamping=true
    bUseGravityOverride=true
    bUsePlanarLimit=false
    gravityScale=0.35
    simulationSpace=RootRelative
    boundBone {boneName=dyn_hair_d_03}
    componentAppliedLinearAccClamp=<1400 1400 1400>
    componentLinearAccScale=<1 1 1>
    componentLinearVelScale=<1 1 1>
    gravityOverride {z=-980}
    physicsBodyDefinitions [{
      boundBone {boneName=dyn_hair_d_03}
      boxExtents=<20 30 30>
      constraintSetup {angularLimitsMax=<3 3 3>, angularLimitsMin=<-3 -3 -3>, angularTarget=<1 0 0>}
      localJointOffset=<4 0 0>
     }]
    simSpaceSettings {maxAngularAcceleration=30, maxAngularVelocity=300, simSpaceAngularAlpha=1}
   }
   n9 = Anim Dynamics - Bone: dyn_hair_e_01(ComponentPose=@n8, Alpha=1, ExternalForce="0,0,0") {
    angularDampingOverride=0.99
    angularSpringConstant=10000
    bAngularSpring=true
    bOverrideAngularDamping=true
    bUseGravityOverride=true
    bUsePlanarLimit=false
    gravityScale=0.35
    simulationSpace=RootRelative
    boundBone {boneName=dyn_hair_e_01}
    componentAppliedLinearAccClamp=<1400 1400 1400>
    componentLinearAccScale=<1 1 1>
    componentLinearVelScale=<1 1 1>
    gravityOverride {z=-980}
    physicsBodyDefinitions [{
      boundBone {boneName=dyn_hair_e_01}
      boxExtents=<20 30 30>
      constraintSetup {angularLimitsMax=<3 3 3>, angularLimitsMin=<-3 -3 -3>, angularTarget=<1 0 0>}
      localJointOffset=<4 0 0>
     }]
    simSpaceSettings {maxAngularAcceleration=30, maxAngularVelocity=300, simSpaceAngularAlpha=1}
   }
   n10 = Anim Dynamics - Bone: dyn_hair_e_02(ComponentPose=@n9, Alpha=1, ExternalForce="0,0,0") {
    angularDampingOverride=0.99
    angularSpringConstant=10000
    bAngularSpring=true
    bOverrideAngularDamping=true
    bUseGravityOverride=true
    bUsePlanarLimit=false
    gravityScale=0.35
    simulationSpace=RootRelative
    boundBone {boneName=dyn_hair_e_02}
    componentAppliedLinearAccClamp=<1400 1400 1400>
    componentLinearAccScale=<1 1 1>
    componentLinearVelScale=<1 1 1>
    gravityOverride {z=-980}
    physicsBodyDefinitions [{
      boundBone {boneName=dyn_hair_e_02}
      boxExtents=<20 30 30>
      constraintSetup {angularLimitsMax=<3 3 3>, angularLimitsMin=<-3 -3 -3>, angularTarget=<1 0 0>}
      localJointOffset=<4 0 0>
     }]
    simSpaceSettings {maxAngularAcceleration=30, maxAngularVelocity=300, simSpaceAngularAlpha=1}
   }
   n11 = Anim Dynamics - Bone: dyn_hair_f_01(ComponentPose=@n10, Alpha=1, ExternalForce="0,0,0") {
    angularDampingOverride=0.99
    angularSpringConstant=10000
    bAngularSpring=true
    bOverrideAngularDamping=true
    bUseGravityOverride=true
    bUsePlanarLimit=false
    gravityScale=0.35
    simulationSpace=RootRelative
    boundBone {boneName=dyn_hair_f_01}
    componentAppliedLinearAccClamp=<1400 1400 1400>
    componentLinearAccScale=<1 1 1>
    componentLinearVelScale=<1 1 1>
    gravityOverride {z=-980}
    physicsBodyDefinitions [{
      boundBone {boneName=dyn_hair_f_01}
      boxExtents=<20 30 30>
      constraintSetup {angularLimitsMax=<3 3 3>, angularLimitsMin=<-3 -3 -3>, angularTarget=<1 0 0>}
      localJointOffset=<4 0 0>
     }]
    simSpaceSettings {maxAngularAcceleration=30, maxAngularVelocity=300, simSpaceAngularAlpha=1}
   }
   n12 = Anim Dynamics - Bone: dyn_hair_f_02(ComponentPose=@n11, Alpha=1, ExternalForce="0,0,0") {
    angularDampingOverride=0.99
    angularSpringConstant=10000
    bAngularSpring=true
    bOverrideAngularDamping=true
    bUseGravityOverride=true
    bUsePlanarLimit=false
    gravityScale=0.35
    simulationSpace=RootRelative
    boundBone {boneName=dyn_hair_f_02}
    componentAppliedLinearAccClamp=<1400 1400 1400>
    componentLinearAccScale=<1 1 1>
    componentLinearVelScale=<1 1 1>
    gravityOverride {z=-980}
    physicsBodyDefinitions [{
      boundBone {boneName=dyn_hair_f_02}
      boxExtents=<20 30 30>
      constraintSetup {angularLimitsMax=<3 3 3>, angularLimitsMin=<-3 -3 -3>, angularTarget=<1 0 0>}
      localJointOffset=<4 0 0>
     }]
    simSpaceSettings {maxAngularAcceleration=30, maxAngularVelocity=300, simSpaceAngularAlpha=1}
   }
   n13 = Anim Dynamics - Bone: dyn_hair_g_01(ComponentPose=@n12, Alpha=1, ExternalForce="0,0,0") {
    angularDampingOverride=0.99
    angularSpringConstant=10000
    bAngularSpring=true
    bOverrideAngularDamping=true
    bUseGravityOverride=true
    bUsePlanarLimit=false
    gravityScale=0.35
    simulationSpace=RootRelative
    boundBone {boneName=dyn_hair_g_01}
    componentAppliedLinearAccClamp=<1400 1400 1400>
    componentLinearAccScale=<1 1 1>
    componentLinearVelScale=<1 1 1>
    gravityOverride {z=-980}
    physicsBodyDefinitions [{
      boundBone {boneName=dyn_hair_g_01}
      boxExtents=<20 30 30>
      constraintSetup {angularLimitsMax=<3 3 3>, angularLimitsMin=<-3 -3 -3>, angularTarget=<1 0 0>}
      localJointOffset=<4 0 0>
     }]
    simSpaceSettings {maxAngularAcceleration=30, maxAngularVelocity=300, simSpaceAngularAlpha=1}
   }
   n14 = Anim Dynamics - Bone: dyn_hair_g_02(ComponentPose=@n13, Alpha=1, ExternalForce="0,0,0") {
    angularDampingOverride=0.99
    angularSpringConstant=10000
    bAngularSpring=true
    bOverrideAngularDamping=true
    bUseGravityOverride=true
    bUsePlanarLimit=false
    gravityScale=0.35
    simulationSpace=RootRelative
    boundBone {boneName=dyn_hair_g_02}
    componentAppliedLinearAccClamp=<1400 1400 1400>
    componentLinearAccScale=<1 1 1>
    componentLinearVelScale=<1 1 1>
    gravityOverride {z=-980}
    physicsBodyDefinitions [{
      boundBone {boneName=dyn_hair_g_02}
      boxExtents=<20 30 30>
      constraintSetup {angularLimitsMax=<3 3 3>, angularLimitsMin=<-3 -3 -3>, angularTarget=<1 0 0>}
      localJointOffset=<4 0 0>
     }]
    simSpaceSettings {maxAngularAcceleration=30, maxAngularVelocity=300, simSpaceAngularAlpha=1}
   }
   n15 = Anim Dynamics - Bone: dyn_hair_h_01(ComponentPose=@n14, Alpha=1, ExternalForce="0,0,0") {
    angularDampingOverride=0.99
    angularSpringConstant=10000
    bAngularSpring=true
    bOverrideAngularDamping=true
    bUseGravityOverride=true
    bUsePlanarLimit=false
    gravityScale=0.35
    simulationSpace=RootRelative
    boundBone {boneName=dyn_hair_h_01}
    componentAppliedLinearAccClamp=<1400 1400 1400>
    componentLinearAccScale=<1 1 1>
    componentLinearVelScale=<1 1 1>
    gravityOverride {z=-980}
    physicsBodyDefinitions [{
      boundBone {boneName=dyn_hair_h_01}
      boxExtents=<20 30 30>
      constraintSetup {angularLimitsMax=<3 3 3>, angularLimitsMin=<-3 -3 -3>, angularTarget=<1 0 0>}
      localJointOffset=<4 0 0>
     }]
    simSpaceSettings {maxAngularAcceleration=30, maxAngularVelocity=300, simSpaceAngularAlpha=1}
   }
   n16 = Anim Dynamics - Bone: dyn_hair_h_02(ComponentPose=@n15, Alpha=1, ExternalForce="0,0,0") {
    angularDampingOverride=0.99
    angularSpringConstant=10000
    bAngularSpring=true
    bOverrideAngularDamping=true
    bUseGravityOverride=true
    bUsePlanarLimit=false
    gravityScale=0.35
    simulationSpace=RootRelative
    boundBone {boneName=dyn_hair_h_02}
    componentAppliedLinearAccClamp=<1400 1400 1400>
    componentLinearAccScale=<1 1 1>
    componentLinearVelScale=<1 1 1>
    gravityOverride {z=-980}
    physicsBodyDefinitions [{
      boundBone {boneName=dyn_hair_h_02}
      boxExtents=<20 30 30>
      constraintSetup {angularLimitsMax=<3 3 3>, angularLimitsMin=<-3 -3 -3>, angularTarget=<1 0 0>}
      localJointOffset=<4 0 0>
     }]
    simSpaceSettings {maxAngularAcceleration=30, maxAngularVelocity=300, simSpaceAngularAlpha=1}
   }
   n17 = Anim Dynamics - Bone: dyn_hair_h_03(ComponentPose=@n16, Alpha=1, ExternalForce="0,0,0") {
    angularDampingOverride=0.99
    angularSpringConstant=10000
    bAngularSpring=true
    bOverrideAngularDamping=true
    bUseGravityOverride=true
    bUsePlanarLimit=false
    gravityScale=0.35
    simulationSpace=RootRelative
    boundBone {boneName=dyn_hair_h_03}
    componentAppliedLinearAccClamp=<1400 1400 1400>
    componentLinearAccScale=<1 1 1>
    componentLinearVelScale=<1 1 1>
    gravityOverride {z=-980}
    physicsBodyDefinitions [{
      boundBone {boneName=dyn_hair_h_03}
      boxExtents=<20 30 30>
      constraintSetup {angularLimitsMax=<3 3 3>, angularLimitsMin=<-3 -3 -3>, angularTarget=<1 0 0>}
      localJointOffset=<4 0 0>
     }]
    simSpaceSettings {maxAngularAcceleration=30, maxAngularVelocity=300, simSpaceAngularAlpha=1}
   }
   n18 = Anim Dynamics - Bone: dyn_hair_i_01(ComponentPose=@n17, Alpha=1, ExternalForce="0,0,0") {
    angularDampingOverride=0.99
    angularSpringConstant=10000
    bAngularSpring=true
    bOverrideAngularDamping=true
    bUseGravityOverride=true
    bUsePlanarLimit=false
    gravityScale=0.35
    simulationSpace=RootRelative
    boundBone {boneName=dyn_hair_i_01}
    componentAppliedLinearAccClamp=<1400 1400 1400>
    componentLinearAccScale=<1 1 1>
    componentLinearVelScale=<1 1 1>
    gravityOverride {z=-980}
    physicsBodyDefinitions [{
      boundBone {boneName=dyn_hair_i_01}
      boxExtents=<20 30 30>
      constraintSetup {angularLimitsMax=<3 3 3>, angularLimitsMin=<-3 -3 -3>, angularTarget=<1 0 0>}
      localJointOffset=<4 0 0>
     }]
    simSpaceSettings {maxAngularAcceleration=30, maxAngularVelocity=300, simSpaceAngularAlpha=1}
   }
   n19 = Anim Dynamics - Bone: dyn_hair_i_02(ComponentPose=@n18, Alpha=1, ExternalForce="0,0,0") {
    angularDampingOverride=0.99
    angularSpringConstant=10000
    bAngularSpring=true
    bOverrideAngularDamping=true
    bUseGravityOverride=true
    bUsePlanarLimit=false
    gravityScale=0.35
    simulationSpace=RootRelative
    boundBone {boneName=dyn_hair_i_02}
    componentAppliedLinearAccClamp=<1400 1400 1400>
    componentLinearAccScale=<1 1 1>
    componentLinearVelScale=<1 1 1>
    gravityOverride {z=-980}
    physicsBodyDefinitions [{
      boundBone {boneName=dyn_hair_i_02}
      boxExtents=<20 30 30>
      constraintSetup {angularLimitsMax=<3 3 3>, angularLimitsMin=<-3 -3 -3>, angularTarget=<1 0 0>}
      localJointOffset=<4 0 0>
     }]
    simSpaceSettings {maxAngularAcceleration=30, maxAngularVelocity=300, simSpaceAngularAlpha=1}
   }
   n20 = Anim Dynamics - Bone: dyn_hair_j_01(ComponentPose=@n19, Alpha=1, ExternalForce="0,0,0") {
    angularDampingOverride=0.99
    angularSpringConstant=10000
    bAngularSpring=true
    bOverrideAngularDamping=true
    bUseGravityOverride=true
    bUsePlanarLimit=false
    gravityScale=0.35
    simulationSpace=RootRelative
    boundBone {boneName=dyn_hair_j_01}
    componentAppliedLinearAccClamp=<1400 1400 1400>
    componentLinearAccScale=<1 1 1>
    componentLinearVelScale=<1 1 1>
    gravityOverride {z=-980}
    physicsBodyDefinitions [{
      boundBone {boneName=dyn_hair_j_01}
      boxExtents=<20 30 30>
      constraintSetup {angularLimitsMax=<3 3 3>, angularLimitsMin=<-3 -3 -3>, angularTarget=<1 0 0>}
      localJointOffset=<4 0 0>
     }]
    simSpaceSettings {maxAngularAcceleration=30, maxAngularVelocity=300, simSpaceAngularAlpha=1}
   }
   n21 = Anim Dynamics - Bone: dyn_hair_j_02(ComponentPose=@n20, Alpha=1, ExternalForce="0,0,0") {
    angularDampingOverride=0.99
    angularSpringConstant=10000
    bAngularSpring=true
    bOverrideAngularDamping=true
    bUseGravityOverride=true
    bUsePlanarLimit=false
    gravityScale=0.35
    simulationSpace=RootRelative
    boundBone {boneName=dyn_hair_j_02}
    componentAppliedLinearAccClamp=<1400 1400 1400>
    componentLinearAccScale=<1 1 1>
    componentLinearVelScale=<1 1 1>
    gravityOverride {z=-980}
    physicsBodyDefinitions [{
      boundBone {boneName=dyn_hair_j_02}
      boxExtents=<20 30 30>
      constraintSetup {angularLimitsMax=<3 3 3>, angularLimitsMin=<-3 -3 -3>, angularTarget=<1 0 0>}
      localJointOffset=<4 0 0>
     }]
    simSpaceSettings {maxAngularAcceleration=30, maxAngularVelocity=300, simSpaceAngularAlpha=1}
   }
   n22 = Anim Dynamics - Bone: dyn_hair_k_01(ComponentPose=@n21, Alpha=1, ExternalForce="0,0,0") {
    angularDampingOverride=0.99
    angularSpringConstant=10000
    bAngularSpring=true
    bOverrideAngularDamping=true
    bUseGravityOverride=true
    bUsePlanarLimit=false
    gravityScale=0.35
    simulationSpace=RootRelative
    boundBone {boneName=dyn_hair_k_01}
    componentAppliedLinearAccClamp=<1400 1400 1400>
    componentLinearAccScale=<1 1 1>
    componentLinearVelScale=<1 1 1>
    gravityOverride {z=-980}
    physicsBodyDefinitions [{
      boundBone {boneName=dyn_hair_k_01}
      boxExtents=<20 30 30>
      constraintSetup {angularLimitsMax=<3 3 3>, angularLimitsMin=<-3 -3 -3>, angularTarget=<1 0 0>}
      localJointOffset=<4 0 0>
     }]
    simSpaceSettings {maxAngularAcceleration=30, maxAngularVelocity=300, simSpaceAngularAlpha=1}
   }
   n23 = Anim Dynamics - Bone: dyn_hair_k_02(ComponentPose=@n22, Alpha=1, ExternalForce="0,0,0") {
    angularDampingOverride=0.99
    angularSpringConstant=10000
    bAngularSpring=true
    bOverrideAngularDamping=true
    bUseGravityOverride=true
    bUsePlanarLimit=false
    gravityScale=0.35
    simulationSpace=RootRelative
    boundBone {boneName=dyn_hair_k_02}
    componentAppliedLinearAccClamp=<1400 1400 1400>
    componentLinearAccScale=<1 1 1>
    componentLinearVelScale=<1 1 1>
    gravityOverride {z=-980}
    physicsBodyDefinitions [{
      boundBone {boneName=dyn_hair_k_02}
      boxExtents=<20 30 30>
      constraintSetup {angularLimitsMax=<3 3 3>, angularLimitsMin=<-3 -3 -3>, angularTarget=<1 0 0>}
      localJointOffset=<4 0 0>
     }]
    simSpaceSettings {maxAngularAcceleration=30, maxAngularVelocity=300, simSpaceAngularAlpha=1}
   }
   n24 = Anim Dynamics - Bone: dyn_hair_l_01(ComponentPose=@n23, Alpha=1, ExternalForce="0,0,0") {
    angularDampingOverride=0.99
    angularSpringConstant=10000
    bAngularSpring=true
    bOverrideAngularDamping=true
    bUseGravityOverride=true
    bUsePlanarLimit=false
    gravityScale=0.35
    simulationSpace=RootRelative
    boundBone {boneName=dyn_hair_l_01}
    componentAppliedLinearAccClamp=<1400 1400 1400>
    componentLinearAccScale=<1 1 1>
    componentLinearVelScale=<1 1 1>
    gravityOverride {z=-980}
    physicsBodyDefinitions [{
      boundBone {boneName=dyn_hair_l_01}
      boxExtents=<20 30 30>
      constraintSetup {angularLimitsMax=<3 3 3>, angularLimitsMin=<-3 -3 -3>, angularTarget=<1 0 0>}
      localJointOffset=<4 0 0>
     }]
    simSpaceSettings {maxAngularAcceleration=30, maxAngularVelocity=300, simSpaceAngularAlpha=1}
   }
   n25 = Anim Dynamics - Bone: dyn_hair_l_02(ComponentPose=@n24, Alpha=1, ExternalForce="0,0,0") {
    angularDampingOverride=0.99
    angularSpringConstant=10000
    bAngularSpring=true
    bOverrideAngularDamping=true
    bUseGravityOverride=true
    bUsePlanarLimit=false
    gravityScale=0.35
    simulationSpace=RootRelative
    boundBone {boneName=dyn_hair_l_02}
    componentAppliedLinearAccClamp=<1400 1400 1400>
    componentLinearAccScale=<1 1 1>
    componentLinearVelScale=<1 1 1>
    gravityOverride {z=-980}
    physicsBodyDefinitions [{
      boundBone {boneName=dyn_hair_l_02}
      boxExtents=<20 30 30>
      constraintSetup {angularLimitsMax=<3 3 3>, angularLimitsMin=<-3 -3 -3>, angularTarget=<1 0 0>}
      localJointOffset=<4 0 0>
     }]
    simSpaceSettings {maxAngularAcceleration=30, maxAngularVelocity=300, simSpaceAngularAlpha=1}
   }
   n26 = Anim Dynamics - Bone: dyn_hair_m_01(ComponentPose=@n25, Alpha=1, ExternalForce="0,0,0") {
    angularDampingOverride=0.99
    angularSpringConstant=10000
    bAngularSpring=true
    bOverrideAngularDamping=true
    bUseGravityOverride=true
    bUsePlanarLimit=false
    gravityScale=0.35
    simulationSpace=RootRelative
    boundBone {boneName=dyn_hair_m_01}
    componentAppliedLinearAccClamp=<1400 1400 1400>
    componentLinearAccScale=<1 1 1>
    componentLinearVelScale=<1 1 1>
    gravityOverride {z=-980}
    physicsBodyDefinitions [{
      boundBone {boneName=dyn_hair_m_01}
      boxExtents=<20 30 30>
      constraintSetup {angularLimitsMax=<3 3 3>, angularLimitsMin=<-3 -3 -3>, angularTarget=<1 0 0>}
      localJointOffset=<4 0 0>
     }]
    simSpaceSettings {maxAngularAcceleration=30, maxAngularVelocity=300, simSpaceAngularAlpha=1}
   }
   n27 = Anim Dynamics - Bone: dyn_hair_m_02(ComponentPose=@n26, Alpha=1, ExternalForce="0,0,0") {
    angularDampingOverride=0.99
    angularSpringConstant=10000
    bAngularSpring=true
    bOverrideAngularDamping=true
    bUseGravityOverride=true
    bUsePlanarLimit=false
    gravityScale=0.35
    simulationSpace=RootRelative
    boundBone {boneName=dyn_hair_m_02}
    componentAppliedLinearAccClamp=<1400 1400 1400>
    componentLinearAccScale=<1 1 1>
    componentLinearVelScale=<1 1 1>
    gravityOverride {z=-980}
    physicsBodyDefinitions [{
      boundBone {boneName=dyn_hair_m_02}
      boxExtents=<20 30 30>
      constraintSetup {angularLimitsMax=<3 3 3>, angularLimitsMin=<-3 -3 -3>, angularTarget=<1 0 0>}
      localJointOffset=<4 0 0>
     }]
    simSpaceSettings {maxAngularAcceleration=30, maxAngularVelocity=300, simSpaceAngularAlpha=1}
   }
   n28 = Anim Dynamics - Bone: dyn_hair_n_01(ComponentPose=@n27, Alpha=1, ExternalForce="0,0,0") {
    angularDampingOverride=0.99
    angularSpringConstant=10000
    bAngularSpring=true
    bOverrideAngularDamping=true
    bUseGravityOverride=true
    bUsePlanarLimit=false
    gravityScale=0.35
    simulationSpace=RootRelative
    boundBone {boneName=dyn_hair_n_01}
    componentAppliedLinearAccClamp=<1400 1400 1400>
    componentLinearAccScale=<1 1 1>
    componentLinearVelScale=<1 1 1>
    gravityOverride {z=-980}
    physicsBodyDefinitions [{
      boundBone {boneName=dyn_hair_n_01}
      boxExtents=<20 30 30>
      constraintSetup {angularLimitsMax=<3 3 3>, angularLimitsMin=<-3 -3 -3>, angularTarget=<1 0 0>}
      localJointOffset=<4 0 0>
     }]
    simSpaceSettings {maxAngularAcceleration=30, maxAngularVelocity=300, simSpaceAngularAlpha=1}
   }
   n29 = Anim Dynamics - Bone: dyn_hair_n_02(ComponentPose=@n28, Alpha=1, ExternalForce="0,0,0") {
    angularDampingOverride=0.99
    angularSpringConstant=10000
    bAngularSpring=true
    bOverrideAngularDamping=true
    bUseGravityOverride=true
    bUsePlanarLimit=false
    gravityScale=0.35
    simulationSpace=RootRelative
    boundBone {boneName=dyn_hair_n_02}
    componentAppliedLinearAccClamp=<1400 1400 1400>
    componentLinearAccScale=<1 1 1>
    componentLinearVelScale=<1 1 1>
    gravityOverride {z=-980}
    physicsBodyDefinitions [{
      boundBone {boneName=dyn_hair_n_02}
      boxExtents=<20 30 30>
      constraintSetup {angularLimitsMax=<3 3 3>, angularLimitsMin=<-3 -3 -3>, angularTarget=<1 0 0>}
      localJointOffset=<4 0 0>
     }]
    simSpaceSettings {maxAngularAcceleration=30, maxAngularVelocity=300, simSpaceAngularAlpha=1}
   }
   Output Pose(Result=Component To Local(ComponentPose=@n29)) {layerGroup=None}
  ]}
 EventGraph {nodes [Event BlueprintUpdateAnimation()]}
]
variables []
